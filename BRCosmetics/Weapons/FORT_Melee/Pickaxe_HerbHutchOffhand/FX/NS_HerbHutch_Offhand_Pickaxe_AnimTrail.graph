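# Niagara animation trail effect for the HerbHutch offhand pickaxe, drawing a short-lived ribbon named Shield_Trail.
# FN anim-trail modules spawn, initialize and update particles from a system-level skeletal mesh and its bone indices.
# Alpha fades along a curve over each particle's life while RGB brightness scales with time of day.
# A User.Variant float compared against 0.5 sets default and alt variant render-enable flags, alongside two ribbon renderers.
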NiagaraSystem /BRCosmetics/Weapons/FORT_Melee/Pickaxe_HerbHutchOffhand/FX/NS_HerbHutch_Offhand_Pickaxe_AnimTrail
dependencies {
 usedDataInterfaces [&NiagaraDataInterfaceCurve, &NiagaraDataInterfaceSkeletalMesh]
 usedDynamicInputs []
 usedModules [
  &self:SystemSpawnScript.SystemScriptSource.SystemScriptGraph.NiagaraNodeAssignment_0.SetVariables_#1
  &SystemState
 ]
}
emitters {Shield_Trail {
  emitterUpdateScript [
   &EmitterState()
   &FN_SpawnAnimTrail()
   /SetVariables(Emitter.RenderEnableDefaultVariant:NiagaraBool=&SetBoolByFloatComparison(A=$User.Variant, B=0.5, Comparison Type=::NewEnumerator6)) {isSetParameters=true}
   /SetVariables(Emitter.RenderEnableAltVariant:NiagaraBool=&SetBoolByFloatComparison(A=$User.Variant, B=0.5)) {isSetParameters=true}
  ]
  particleSpawnScript [
   &FN_InitializeAnimTrailBoneIndecies()
   &FN_InitializeAnimTrail(Color=<rgba 1 1 1 0.8>, Lifetime=0.12, SkeletalMesh=$System.Skeletal Mesh)
  ]
  particleUpdateScript [
   &SolveForcesAndVelocity(Clamp Velocity=false, Limit Acceleration=false)
   &FN_UpdateAnimTrail(SkeletalMesh=$System.Skeletal Mesh) {module=FN_UpdateAnimTrail001}
   &ScaleColor(
    Scale Alpha=&FloatFromCurve(FloatCurve=DataInterface() {propertyValues {Curve {keys [<at 0 1 RCIM_Cubic>,<at 1 0>]}}})
    Scale RGB=&VectorFromFloat(Value=&FNTimeOfDayScalar(Night Scale=0.65, Morning Scale=0.8, Day Scale=0.9, Evening Scale=0.8))
    ScaleA=true
    ScaleRGB=true
   ) {module=ScaleColor001}
  ]
  rendererClasses [&NiagaraRibbonRendererProperties]
  renderers [
   {rendererClass=&NiagaraRibbonRendererProperties}
   {rendererClass=&NiagaraRibbonRendererProperties, rendererIndex=1}
  ]
 }}
summary {
 systemName=NS_HerbHutch_Offhand_Pickaxe_AnimTrail
 emitters [{emitterName=Shield_Trail, rendererClasses [&NiagaraRibbonRendererProperties]}]
 userVariables [
  User.RibbonIDOffset {type=NiagaraInt32}
  User.Variant {type=NiagaraFloat, defaultValue=0, description=""}
 ]
}
