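# Physics asset for the HerbHutchOffhand pickaxe axe mesh, a cosmetic melee weapon.
# It holds a single body on the Melee_TwoHanded_Handle bone, built from one large rotated capsule offset from the bone origin.
# No constraints exist, so the body provides simple collision bounds rather than articulated simulation.
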PhysicsAsset /BRCosmetics/Weapons/FORT_Melee/Pickaxe_HerbHutchOffhand/Meshes/HerbHutchOffhand_Axe_PhysicsAsset
bodies {Melee_TwoHanded_Handle [{shapeType=Capsule, length=25.7369, radius=25.7368, center=<2.7078 -8.1042 28.1686>, rotation=<rot 90 -38.1751 51.8052>}]}
constraints []
